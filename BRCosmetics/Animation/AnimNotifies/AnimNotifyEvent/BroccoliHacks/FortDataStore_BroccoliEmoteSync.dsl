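; FortDataStore_BroccoliEmoteSync is a data store that records the cosmetic channel index active when an emote sync anim notify starts.
; It keeps a map from object references to integer channel indices.
; PushNotifyStartChannel adds an object's index to the map, and GetNotifyStartChannel looks it up, returning -1 when no entry exists.
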(fn GetNotifyStartChannel
  (params (For :type Object%20Reference))
  (bind _cosmeticchannelatnotifystart (Variables|Default|GetCosmeticChannelAtNotifyStart))
  (bind (_value _returnvalue) (Utilities|Map|Find _cosmeticchannelatnotifystart For))
  (if _returnvalue
    (return _value)
    (else
      (return -1))))

(fn PushNotifyStartChannel
  (params (For :type Object%20Reference) (ChannelIdx :type Integer))
  (Utilities|Map|Add (Variables|Default|GetCosmeticChannelAtNotifyStart) For ChannelIdx))
